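# Physics asset for the face accessory of the male medium ClayPlug Mask body, covering its left and right mask tubes.
# Each dynamic tube capsule hangs off a tiny base-tube body chained to the backpack, with narrow angular limits at the tube joint.
# Backpack, head and clavicle capsules approximate the surrounding body, linking to spine_05 directly or through neck_01 by constraints with free swing and twist.
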PhysicsAsset /BRCosmetics/Characters/Player/Male/Medium/Bodies/M_MED_ClayPlug_Mask/Meshes/Parts/M_MED_ClayPlug_Mask_FaceAcc_Physics
bodies {
 backpack_c [{shapeType=Capsule, length=19.3836, radius=11.841, center=<0.0018 -15.7839 -2.6202>, rotation=<rot 0 90.0019 -89.9972>}]
 base_mask_tubes_01_l [{shapeType=Capsule, length=0.505, radius=0.505, shapeName=base_mask_tubes_01_l_capsule, rotation=<rot 0 -90 0>}]
 base_mask_tubes_01_r [{shapeType=Capsule, length=0.505, radius=0.505, shapeName=base_mask_tubes_01_r_capsule, rotation=<rot 0 -90 0>}]
 clavicle_l [{shapeType=Capsule, length=5.6712, radius=8.3519, shapeName=clavicle_l_capsule, center=<10.2203 0 -1.6867>, rotation=<rot -0.0002 -90 90>}]
 clavicle_r [{shapeType=Capsule, length=5.6712, radius=8.3519, shapeName=clavicle_l_capsule, center=<-10.2203 0 1.6867>, rotation=<rot -0.0002 -90 -90>}]
 dyn_mask_tubes_01_l [{shapeType=Capsule, length=10.8176, radius=2.9827, shapeName=dyn_mask_tubes_01_l_capsule, center=<30 0 -9.9484>, rotation=<rot 90 0 0>}]
 dyn_mask_tubes_01_r [{shapeType=Capsule, length=10.8176, radius=2.8951, shapeName=dyn_mask_tubes_01_r_capsule, center=<16.6959 2.0101 0>, rotation=<rot 90 0 0>}]
 head [{shapeType=Capsule, length=2.9535, radius=8.7268, center=<5.7498 1.2213 0>, rotation=<rot 90 -90 -89.9998>}]
 neck_01 [{shapeType=Capsule, length=0.505, radius=0.505, shapeName=neck_01_capsule, rotation=<rot 0 -90 0>}]
 spine_05 [{shapeType=Capsule, length=0.505, radius=0.505, shapeName=spine_05_capsule, rotation=<rot 0 -90 0>}]
}
constraints [
 {
  bone1Name=head
  bone2Name=neck_01
  swing1LimitDegrees=45
  swing1Motion=Free
  swing2LimitDegrees=45
  swing2Motion=Free
  twistLimitDegrees=45
  twistMotion=Free
 }
 {
  bone1Name=backpack_c
  bone2Name=spine_05
  swing1LimitDegrees=45
  swing1Motion=Free
  swing2LimitDegrees=45
  swing2Motion=Free
  twistLimitDegrees=45
  twistMotion=Free
 }
 {bone1Name=dyn_mask_tubes_01_l, bone2Name=base_mask_tubes_01_l, swing1LimitDegrees=0, swing2LimitDegrees=10, twistLimitDegrees=0}
 {
  bone1Name=base_mask_tubes_01_l
  bone2Name=backpack_c
  swing1LimitDegrees=45
  swing1Motion=Free
  swing2LimitDegrees=45
  swing2Motion=Free
  twistLimitDegrees=45
  twistMotion=Free
 }
 {bone1Name=dyn_mask_tubes_01_r, bone2Name=base_mask_tubes_01_r, swing1LimitDegrees=0, swing2LimitDegrees=10, twistLimitDegrees=0}
 {
  bone1Name=base_mask_tubes_01_r
  bone2Name=backpack_c
  swing1LimitDegrees=45
  swing1Motion=Free
  swing2LimitDegrees=45
  swing2Motion=Free
  twistLimitDegrees=45
  twistMotion=Free
 }
 {
  bone1Name=clavicle_l
  bone2Name=spine_05
  swing1LimitDegrees=45
  swing1Motion=Free
  swing2LimitDegrees=45
  swing2Motion=Free
  twistLimitDegrees=45
  twistMotion=Free
 }
 {
  bone1Name=neck_01
  bone2Name=spine_05
  swing1LimitDegrees=45
  swing1Motion=Free
  swing2LimitDegrees=45
  swing2Motion=Free
  twistLimitDegrees=45
  twistMotion=Free
 }
 {
  bone1Name=clavicle_r
  bone2Name=spine_05
  swing1LimitDegrees=45
  swing1Motion=Free
  swing2LimitDegrees=45
  swing2Motion=Free
  twistLimitDegrees=45
  twistMotion=Free
 }
]
