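# Post-process style AnimBP for the ClayPlug Mask character's face accessory, animating mask tubes, backpack and hair-correction bones.
# Copies the attached parent's pose, layers body-mesh clavicles and upper arms, then runs Control Rig, additive bone offsets and spine_05-space RigidBody physics.
# The update event derives spine-twist backpack offsets, skydiving states and mask/backpack tag flags.
# On customization load, a ClayPlug backpack tag check toggles a material hide parameter.
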AnimBlueprint /BRCosmetics/Characters/Player/Male/Medium/Bodies/M_MED_ClayPlug_Mask/Meshes/Parts/M_MED_ClayPlug_Mask_FaceAcc_AnimBP
target_skeleton=&/M_MED_ClayPlug_Mask_FaceAcc_Skeleton
graphs [
 AnimGraph {nodes [
   n1 = "Transform (Modify) Bone - Bone: backpack_c"!(ComponentPose=Local To Component(LocalPose=Use cached pose 'StartPose'()), Alpha=1, Translation="0,0,0", Rotation="0,0,0") -> (Pose) {rotationMode=BMM_Additive, translationMode=BMM_Additive, alphaBoolBlend {blendOption=Linear}, boneToModify {boneName=backpack_c}}
   n2 = "Transform (Modify) Bone - Bone: hair_correction_c"!(ComponentPose=@n1, bAlphaBoolEnabled:bool=false, Translation="-0.003,-0.471,-0.107", Rotation="0,0,0") -> (Pose) {
    alphaInputType=Bool
    rotationMode=BMM_Additive
    rotationSpace=BCS_BoneSpace
    translationMode=BMM_Additive
    translationSpace=BCS_BoneSpace
    alphaBoolBlend {blendOption=Linear}
    boneToModify {boneName=hair_correction_c}
    translation=<-0.003 -0.471 -0.107>
   }
   n3 = Control Rig(Source=Component To Local(ComponentPose=@n2)) {
    alphaBoolBlend {blendOption=Linear}
    controlRigAssetReference {blueprintRigClass=&/M_MED_ClayPlug_Mask_FaceAcc_CtrlRig.M_MED_ClayPlug_Mask_FaceAcc_CtrlRig_C}
   }
   n4 = "Transform (Modify) Bone - Bone: backpack_c"!(ComponentPose=Local To Component(LocalPose=@n3), bAlphaBoolEnabled:bool=True, Translation="0,-1.5,2", Rotation="0,0,0") -> (Pose) {
    alphaInputType=Bool
    rotationMode=BMM_Additive
    rotationSpace=BCS_BoneSpace
    translationMode=BMM_Additive
    translationSpace=BCS_BoneSpace
    alphaBoolBlend {blendInTime=2, blendOption=Linear, blendOutTime=2}
    alphaScaleBiasClamp {bInterpResult=true}
    boneToModify {boneName=backpack_c}
    translation {y=-1.5, z=2}
   }
   n5 = "Transform (Modify) Bone - Bone: base_mask_tubes_01_l"!(ComponentPose=@n4, Alpha=1, Translation="0,0,0", Rotation="0,0,0") -> (Pose) {
    rotationMode=BMM_Additive
    rotationSpace=BCS_BoneSpace
    translationMode=BMM_Additive
    translationSpace=BCS_BoneSpace
    alphaBoolBlend {blendInTime=2, blendOption=Linear, blendOutTime=2}
    alphaScaleBiasClamp {bInterpResult=true}
    boneToModify {boneName=base_mask_tubes_01_l}
   }
   n6 = "Transform (Modify) Bone - Bone: base_mask_tubes_01_r"!(ComponentPose=@n5, Alpha=1, Translation="0,0,0", Rotation="-4.316,-2.890,6.838") -> (Pose) {
    rotationMode=BMM_Additive
    rotationSpace=BCS_BoneSpace
    translationMode=BMM_Additive
    translationSpace=BCS_BoneSpace
    alphaBoolBlend {blendInTime=2, blendOption=Linear, blendOutTime=2}
    alphaScaleBiasClamp {bInterpResult=true}
    boneToModify {boneName=base_mask_tubes_01_r}
    rotation=<rot -4.316 -2.8901 6.8381>
   }
   n7 = "Transform (Modify) Bone - Bone: mask_tubes_02_l"!(ComponentPose=@n6, Alpha=1, Translation="0,0,0", Rotation="-29.911,-10.796,2.933") -> (Pose) {
    rotationMode=BMM_Additive
    rotationSpace=BCS_BoneSpace
    translationMode=BMM_Additive
    translationSpace=BCS_BoneSpace
    alphaBoolBlend {blendInTime=2, blendOption=Linear, blendOutTime=2}
    alphaScaleBiasClamp {bInterpResult=true}
    boneToModify {boneName=mask_tubes_02_l}
    rotation=<rot -29.911 -10.796 2.933>
   }
   n8 = "Transform (Modify) Bone - Bone: backpack_c"!(ComponentPose=@n7, bAlphaBoolEnabled=True, Translation="0,-1.5,0", Rotation="0,0,0", Scale:Vector="1,1,1") -> (Pose) {
    alphaInputType=Bool
    rotationMode=BMM_Additive
    rotationSpace=BCS_BoneSpace
    translationMode=BMM_Additive
    translationSpace=BCS_BoneSpace
    alphaBoolBlend {blendInTime=0.2, blendOption=Linear, blendOutTime=0.2}
    boneToModify {boneName=backpack_c}
    translation {y=-1.5}
   }
   n9 = "Transform (Modify) Bone - Bone: backpack_c"!(ComponentPose=@n8, Alpha:float=1, Translation="-0.000,-1.270,3.449", Rotation="-0.000,0.000,13.598", Scale="1,1,1") -> (Pose) {
    rotationMode=BMM_Additive
    rotationSpace=BCS_BoneSpace
    translationMode=BMM_Additive
    translationSpace=BCS_BoneSpace
    alphaBoolBlend {blendInTime=0.2, blendOption=Linear, blendOutTime=0.2}
    alphaScaleBiasClamp {bInterpResult=true, interpSpeedDecreasing=3, interpSpeedIncreasing=5}
    boneToModify {boneName=backpack_c}
    rotation=<rot 0 0 13.598>
    translation {y=-1.27, z=3.449}
   }
   n10 = RigidBody(ComponentPose=@n9) {
    bClampLinearTranslationLimitToRefPose=true
    bOverrideWorldGravity=true
    overridePhysicsAsset=&/M_MED_ClayPlug_Mask_FaceAcc_Physics
    simulationSpace=BaseBoneSpace
    alphaBoolBlend {blendOption=Linear}
    alphaScaleBiasClamp {bInterpResult=true}
    baseBoneRef {boneName=spine_05}
    componentLinearAccScale=<0.5 0.2 0.2>
    componentLinearVelScale=<0.5 0.2 0.2>
    overrideWorldGravity {z=-980}
    simSpaceSettings {velocityScaleZ=0.2}
   }
   Output Pose(Result=Component To Local(ComponentPose=@n10)) {layerGroup=None}
   n11 = Layered blend per bone(
    BasePose=Copy Pose From Mesh() {bCopyCurves=true, bUseAttachedParent=true}
    BlendPoses_0=Copy Pose From Mesh(SourceMeshComponent=Get Body_Mesh())
   ) {layerSetup [{branchFilters [{boneName=clavicle_l}, {boneName=clavicle_r}, {boneName=upperarm_r}, {boneName=upperarm_l}]}]}
   Save cached pose 'StartPose'(Pose=@n11) {cacheName=StartPose}
  ]}
 EventGraph {nodes [
   Event OnCharacterCustomizationLoaded()
   > n1 = Sequence!(execute) -> (then_0, then_1, then_2:Exec)
   > n2 = Cast To PlayerPawn_Athena(Object=TryGetPawnOwner())
   > Set Body_Mesh(Body_Mesh=GetSkeletalMeshForPartType(self=@n2))
   n3 = Get Dynamic Material Instance(execute=@n1.then_2, SkeletalMeshComponent=GetPartSkeletalMeshComponent())
   > n4 = Branch(Condition=BackpackItemDefinitionHasAnyMetaTag!(Tags:Array of Gameplay Tag Structures=Make Array("[0]"=(TagName="Cosmetics.MetaTags.ClayPlug")), bRequiresEquipped:bool=false) -> (ReturnValue:bool))
   > SetScalarParameterValue(self=@n3, ParameterName=Hide Element 01, Value=0)
   SetScalarParameterValue(execute=@n4.else, self=@n3, ParameterName=Hide Element 01, Value=1)
   Event BlueprintUpdateAnimation()
   > n5 = Sequence!(execute) -> (then_0, then_1, then_2, then_3:Exec, then_4:Exec, then_5:Exec)
   > n6 = Sequence!(execute) -> (then_0, then_1)
   n7 = GetOwnerSkeletalMeshComponent()
   n8 = BreakRotator(InRot=BreakTransform(InTransform=GetDeltaTransformFromRefPose(self=@n7, BoneName=spine_05, BaseName=spine_01)).Rotation)
   > Set Backpack Rotate(Backpack Rotate:Rotator=MakeRotator(Roll=MapRangeClamped(Value=@n8.Yaw, InRangeA=0, InRangeB=-80, OutRangeB=20))) -> (Output_Get:Rotator)
   Set Backpack Translation(execute=@n6.then_1, Backpack Translation:Vector=MakeVector(Y=MapRangeClamped(Value=@n8.Yaw, InRangeA=0, InRangeB=-30, OutRangeB=-1.9))) -> (Output_Get:Vector)
   n9 = Get bIsSkyDiving()
   Set Skydive Alpha(execute=@n5.then_1, Skydive Alpha:double="To Float (Boolean)"(InBool=NOT Boolean(A=@n9))) -> (Output_Get:double)
   n10 = Get bIsSkydiveFloating()
   n11 = Get bIsSkydiveDiveMode()
   Set Skydive Rigid Alpha(execute=@n5.then_2, Skydive Rigid Alpha:double="To Float (Boolean)"(InBool=OR Boolean(A=@n10, B=@n11))) -> (Output_Get:double)
   n12 = Make Array("[0]"=(TagName="Cosmetics.MetaTags.ClayPlug"))
   Set skydiveClayPlug(
    execute=@n5.then_3
    skydiveClayPlug:bool=AND Boolean(A=Get bIsSkydiveFloating(), B=BackpackItemDefinitionHasAnyMetaTag(Tags=@n12, bRequiresEquipped=false))
   ) -> (Output_Get:bool)
   Set IsMaskOn(execute=@n5.then_5, IsMaskOn:bool=CharacterItemDefinitionHasAnyMetaTag(Tags=Make Array("[0]"=(TagName="Cosmetics.MetaTags.Mask.On")))) -> (Output_Get:bool)
   n13 = GetOwnerSkeletalMeshComponent()
   n14 = BreakRotator(InRot=BreakTransform(InTransform=GetDeltaTransformFromRefPose(self=@n13, BoneName=pelvis, BaseName=root)).Rotation)
   n15 = Branch(
    execute=@n5.then_4
    Condition=AND Boolean!(A="float > float"(A=@n14.Yaw, B=90), B="float < float"(A=@n14.Yaw, B=130), C:bool=Get bIsSkydiveDiveMode()) -> (ReturnValue)
   )
   > Set AlphaIsCorrectiveSkydiveMode(AlphaIsCorrectiveSkydiveMode:double=1) -> (Output_Get:double)
   Set AlphaIsCorrectiveSkydiveMode(execute=@n15.else, AlphaIsCorrectiveSkydiveMode=0)
   Event OnBeginSkydiving()
   > Set Skydive Alpha(Skydive Alpha=0.1)
   Event OnEndSkydiving()
   > Set Skydive Alpha(Skydive Alpha=1)
  ]}
]
variables [
 AlphaIsCorrectiveSkydiveMode {type=number}
 Backpack Rotate {type=Rotator}
 Backpack Translation {type=Vector}
 Body_Mesh {type=SkeletalMeshComponent}
 IsMaskOn {type=boolean}
 Skydive Alpha {type=number, default=1}
 Skydive Rigid Alpha {type=number}
 skydiveClayPlug {type=boolean}
]
